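# Sound cue for the looping first-person music of the Dashing emote.
# After a short fixed delay, a mixer layers two looping waves: the main Dashing loop at full volume and a quieter saddle loop.
# It uses first-person emote music class, property sheet and concurrency assets, and also sends to a music audio-analysis submix.
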SoundCue /BRCosmetics/Sounds/Emotes/Dashing/Emote_Dashing_Music_Cue
root=Delay_0
nodes [
 {id=Delay_0, children [Mixer_0], properties {delayMax=0.35, delayMin=0.35}}
 {id=Mixer_0, children [WavePlayer_2,WavePlayer_6], properties {inputVolume [1,0.4]}}
 {id=WavePlayer_2, properties {bLooping=true, soundWaveAssetPtr=&/s23_Emote_Dashing_Loop}}
 {id=WavePlayer_6, properties {bLooping=true, soundWaveAssetPtr=&/Emote_Dashing_Saddle_Loop}}
]
properties {
 audioPropertiesSheet=/BRCosmetics/Sounds/Settings/PropertySheets/Emotes/APS_Cosmetic_Emote_Music_1P
 maxDistance=2097152
 soundClassObject=/Game/Sounds/FortSoundClassesAndModes/Fort_Music_Emote_1P_PSM
 volumeMultiplier=1
 concurrencyOverrides {maxCount=3, resolutionRule=StopOldest}
 concurrencySet [/Game/Sounds/Concurrency_Rules/Athena_EmoteMusic_1P_Conc]
 soundSubmixSends [{minSendLevel=0, soundSubmix=&Fortnite_AudioAnalysis_Music}]
}
