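# Third-person looping sound cue for the Dashing emote, routed through the 3P emote music sound class.
# After a short fixed delay, a mixer layers the main emote loop with a quieter saddle loop.
# It uses shared emote attenuation and music concurrency, plays when silent to keep running while inaudible, and has a send to the music audio-analysis submix.
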SoundCue /BRCosmetics/Sounds/Emotes/Dashing/SC_Emote_Dashin_3P_Cue
root=Delay_3
nodes [
 {id=Delay_3, children [Mixer_1], properties {delayMax=0.35, delayMin=0.35}}
 {id=Mixer_1, children [WavePlayer_5,WavePlayer_6], properties {inputVolume [1,0.4]}}
 {id=WavePlayer_5, properties {bLooping=true, soundWaveAssetPtr=&/s23_Emote_Dashing_Loop}}
 {id=WavePlayer_6, properties {bLooping=true, soundWaveAssetPtr=&/Emote_Dashing_Saddle_Loop}}
]
properties {
 attenuationSettings=&Fort_Emotes_Athena_Att
 maxDistance=1600
 soundClassObject=/Game/Sounds/FortSoundClassesAndModes/Fort_Music_Emote_3P
 virtualizationMode=PlayWhenSilent
 volumeMultiplier=1
 concurrencyOverrides {resolutionRule=PreventNew}
 concurrencySet [/Game/Sounds/Concurrency_Rules/Athena_EmoteMusic_Conc]
 soundSubmixSends [{minSendLevel=0, soundSubmix=&Fortnite_AudioAnalysis_Music}]
}
